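# GrillMoon_pjw1000 is a Level Sequence that acts as the parent for the pjw1000 cinematic sequence.
# It layers seven department subsequence tracks (animation, audio, camera, gameplay, lighting, techart, vfx), all spanning the same frame range, at a 60 fps display rate.
# It also binds a spawned static mesh actor, _pjw1000_MoverCube, with a constant world transform and a visibility track.
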LevelSequence /GrillMoon/Sequence/pjw/pjw1000/GrillMoon_pjw1000
bindings [_pjw1000_MoverCube {
  id=#1
  possessed_class=FortStaticMeshActor
  tracks [
   {type=MovieSceneSpawnTrack, sections [{type=MovieSceneSpawnSection, channels [None {type=MovieSceneScriptingBoolChannel, default=true}]}]}
   {
    type=MovieScene3DTransformTrack
    sections [{
      type=MovieScene3DTransformSection
      channels [
       Location.X {type=MovieSceneScriptingDoubleChannel, default=-57346.7329, const [-57594.2109,1]}
       Location.Y {type=MovieSceneScriptingDoubleChannel, default=41986.2606, const [42656.1797,1]}
       Location.Z {type=MovieSceneScriptingDoubleChannel, default=9934.7451, const [9936.5957,1]}
       Rotation.X {type=MovieSceneScriptingDoubleChannel, default=0, const [0,1]}
       Rotation.Y {type=MovieSceneScriptingDoubleChannel, default=0, const [0,1]}
       Rotation.Z {type=MovieSceneScriptingDoubleChannel, default=0, const [0,1]}
       Scale.X {type=MovieSceneScriptingDoubleChannel, default=1, const [1,1]}
       Scale.Y {type=MovieSceneScriptingDoubleChannel, default=1, const [1,1]}
       Scale.Z {type=MovieSceneScriptingDoubleChannel, default=1, const [1,1]}
      ]
     }]
   }
   {
    type=MovieSceneVisibilityTrack
    sections [{type=MovieSceneVisibilitySection, channels [None {type=MovieSceneScriptingBoolChannel, default=false}]}]
   }
  ]
 }]
display_rate {denominator=1, numerator=60}
tick_resolution {denominator=1, numerator=24000}
tracks [
 {type=MovieSceneSubTrack, display_name=animation, sections [{type=MovieSceneSubSection, end_frame=6731, start_frame=3388}]}
 {type=MovieSceneSubTrack, display_name=audio, sections [{type=MovieSceneSubSection, end_frame=6731, start_frame=3388}]}
 {type=MovieSceneSubTrack, display_name=camera, sections [{type=MovieSceneSubSection, end_frame=6731, start_frame=3388}]}
 {type=MovieSceneSubTrack, display_name=gameplay, sections [{type=MovieSceneSubSection, end_frame=6731, start_frame=3388}]}
 {type=MovieSceneSubTrack, display_name=lighting, sections [{type=MovieSceneSubSection, end_frame=6731, start_frame=3388}]}
 {type=MovieSceneSubTrack, display_name=techart, sections [{type=MovieSceneSubSection, end_frame=6731, start_frame=3388}]}
 {type=MovieSceneSubTrack, display_name=vfx, sections [{type=MovieSceneSubSection, end_frame=6731, start_frame=3388}]}
]
