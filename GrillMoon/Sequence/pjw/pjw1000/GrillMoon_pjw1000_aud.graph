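# GrillMoon_pjw1000_aud is a Level Sequence holding the audio layer for GrillMoon cinematic shot pjw1000.
# Root-level audio tracks layer PolishedJade voice and body sounds, MistMylar footsteps and energy, UnluckyRoll sounds and a generic track in staggered, overlapping sections.
# Mix levels are set per section, with keyed volume on one UnluckyRoll clip.
# Possessable bindings to mannequins, an animated mask mesh and maskmaker Niagara effects carry no tracks.
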LevelSequence /GrillMoon/Sequence/pjw/pjw1000/GrillMoon_pjw1000_aud
bindings [
 GrillMoon_EventZone_UnluckyRoll {id=#1, possessed_class=FortPlayerMannequin, tracks []}
 animMask12_Anim {id=#2, possessed_class=FortSkeletalMeshActor, tracks []}
 NS_GrillMoon_Maskmaker_OnFace {id=#3, possessed_class=NiagaraActor, tracks []}
 NS_GrillMoon_Maskmaker_Pull {id=#4, possessed_class=NiagaraActor, tracks []}
 NS_GrillMoon_MasksMerge_Chargeup {id=#5, possessed_class=NiagaraActor, tracks []}
 Char_PolishedJade {id=#6, possessed_class=FortPlayerMannequin, tracks []}
 GrillMoon_Countdown_MistMylar {id=#7, possessed_class=FortPlayerMannequin, tracks []}
]
display_rate {denominator=1, numerator=60}
tick_resolution {denominator=1, numerator=24000}
tracks [
 {
  type=MovieSceneAudioTrack
  display_name=PolishedJade VO
  sections [{
    type=MovieSceneAudioSection
    end_frame=5851
    start_frame=5691
    channels [
     Volume {type=MovieSceneScriptingFloatChannel, default=0.5}
     Pitch {type=MovieSceneScriptingFloatChannel, default=1}
     AttachActor {type=MovieSceneScriptingActorReferenceChannel}
    ]
   }]
 }
 {
  type=MovieSceneAudioTrack
  display_name=MistMylar FS
  sections [
   {
    type=MovieSceneAudioSection
    end_frame=4696
    start_frame=4590
    channels [
     Volume {type=MovieSceneScriptingFloatChannel, default=1}
     Pitch {type=MovieSceneScriptingFloatChannel, default=1}
     AttachActor {type=MovieSceneScriptingActorReferenceChannel}
    ]
   }
   {
    type=MovieSceneAudioSection
    end_frame=4868
    start_frame=4776
    channels [
     Volume {type=MovieSceneScriptingFloatChannel, default=1}
     Pitch {type=MovieSceneScriptingFloatChannel, default=1}
     AttachActor {type=MovieSceneScriptingActorReferenceChannel}
    ]
   }
   {
    type=MovieSceneAudioSection
    end_frame=4901
    start_frame=4836
    channels [
     Volume {type=MovieSceneScriptingFloatChannel, default=1}
     Pitch {type=MovieSceneScriptingFloatChannel, default=1}
     AttachActor {type=MovieSceneScriptingActorReferenceChannel}
    ]
   }
   {
    type=MovieSceneAudioSection
    end_frame=5045
    start_frame=4955
    channels [
     Volume {type=MovieSceneScriptingFloatChannel, default=1}
     Pitch {type=MovieSceneScriptingFloatChannel, default=1}
     AttachActor {type=MovieSceneScriptingActorReferenceChannel}
    ]
   }
  ]
 }
 {
  type=MovieSceneAudioTrack
  display_name=PolishedJade Body
  sections [{
    type=MovieSceneAudioSection
    end_frame=6145
    start_frame=5000
    channels [
     Volume {type=MovieSceneScriptingFloatChannel, default=1}
     Pitch {type=MovieSceneScriptingFloatChannel, default=1}
     AttachActor {type=MovieSceneScriptingActorReferenceChannel}
    ]
   }]
 }
 {
  type=MovieSceneAudioTrack
  display_name=UnluckyRoll
  sections [
   {
    type=MovieSceneAudioSection
    end_frame=6057
    start_frame=5717
    channels [
     Volume {type=MovieSceneScriptingFloatChannel, default=1.3, keys [[5797,1],[5851,1.3]]}
     Pitch {type=MovieSceneScriptingFloatChannel, default=1}
     AttachActor {type=MovieSceneScriptingActorReferenceChannel}
    ]
   }
   {
    type=MovieSceneAudioSection
    end_frame=7027
    start_frame=6145
    channels [
     Volume {type=MovieSceneScriptingFloatChannel, default=1}
     Pitch {type=MovieSceneScriptingFloatChannel, default=1}
     AttachActor {type=MovieSceneScriptingActorReferenceChannel}
    ]
   }
  ]
 }
 {
  type=MovieSceneAudioTrack
  display_name=UnluckyRoll
  sections [{
    type=MovieSceneAudioSection
    end_frame=6840
    start_frame=6145
    channels [
     Volume {type=MovieSceneScriptingFloatChannel, default=0.7}
     Pitch {type=MovieSceneScriptingFloatChannel, default=1}
     AttachActor {type=MovieSceneScriptingActorReferenceChannel}
    ]
   }]
 }
 {
  type=MovieSceneAudioTrack
  display_name=PolishedJade VO
  sections [
   {
    type=MovieSceneAudioSection
    end_frame=5647
    start_frame=5417
    channels [
     Volume {type=MovieSceneScriptingFloatChannel, default=0.5}
     Pitch {type=MovieSceneScriptingFloatChannel, default=1}
     AttachActor {type=MovieSceneScriptingActorReferenceChannel}
    ]
   }
   {
    type=MovieSceneAudioSection
    end_frame=6190
    start_frame=5898
    channels [
     Volume {type=MovieSceneScriptingFloatChannel, default=1}
     Pitch {type=MovieSceneScriptingFloatChannel, default=1}
     AttachActor {type=MovieSceneScriptingActorReferenceChannel}
    ]
   }
  ]
 }
 {type=MovieSceneAudioTrack, display_name=PolishedJade Body, sections []}
 {
  type=MovieSceneAudioTrack
  display_name=UnluckyRoll
  sections [{
    type=MovieSceneAudioSection
    end_frame=6411
    start_frame=5994
    channels [
     Volume {type=MovieSceneScriptingFloatChannel, default=1}
     Pitch {type=MovieSceneScriptingFloatChannel, default=1}
     AttachActor {type=MovieSceneScriptingActorReferenceChannel}
    ]
   }]
 }
 {
  type=MovieSceneAudioTrack
  display_name=MistMylar Energy
  sections [{
    type=MovieSceneAudioSection
    end_frame=6943
    start_frame=6145
    channels [
     Volume {type=MovieSceneScriptingFloatChannel, default=1}
     Pitch {type=MovieSceneScriptingFloatChannel, default=1}
     AttachActor {type=MovieSceneScriptingActorReferenceChannel}
    ]
   }]
 }
 {
  type=MovieSceneAudioTrack
  display_name=MistMylar Energy
  sections [{
    type=MovieSceneAudioSection
    end_frame=6090
    start_frame=4955
    channels [
     Volume {type=MovieSceneScriptingFloatChannel, default=1}
     Pitch {type=MovieSceneScriptingFloatChannel, default=1}
     AttachActor {type=MovieSceneScriptingActorReferenceChannel}
    ]
   }]
 }
 {
  type=MovieSceneAudioTrack
  display_name=Audio
  sections [{
    type=MovieSceneAudioSection
    end_frame=4906
    start_frame=4479
    channels [
     Volume {type=MovieSceneScriptingFloatChannel, default=1}
     Pitch {type=MovieSceneScriptingFloatChannel, default=1}
     AttachActor {type=MovieSceneScriptingActorReferenceChannel}
    ]
   }]
 }
]
