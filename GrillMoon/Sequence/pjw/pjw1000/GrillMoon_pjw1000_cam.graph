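# Camera shake and controller rumble sub-sequence for the GrillMoon pjw1000 event sequence; it holds no camera or cut tracks.
# A spawned Events_Shared_cameraForce_BP at a fixed transform has keyed property tracks that swap its force feedback effect and shake class, including demon walk, charge build-up and end moments.
# Lobby event triggers fire alongside those swaps, with extra shake triggers between them, while each Radius event track has a single trigger.
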LevelSequence /GrillMoon/Sequence/pjw/pjw1000/GrillMoon_pjw1000_cam
bindings [Events_Shared_cameraForce_BP {
  id=#1
  possessed_class=Events_Shared_cameraForce_BP_C
  tracks [
   {
    type=MovieSceneObjectPropertyTrack
    display_name=Force Effect
    sections [{
      type=MovieSceneObjectPropertySection
      channels [None {
        type=MovieSceneScriptingObjectPathChannel
        default="<Object '/GrillMoon/ForcedFeedback/GrillMoon_FF_pjw1000_Demon_Charge_BuildUp' (0x8509b575080) Class 'ForceFeedbackEffect'>"
        keys [
         [0, "<Object '&Events_Shared_small_FF' (0x84c3c703480) Class 'ForceFeedbackEffect'>"]
         [4965, "<Object '/Game/Weapons/ForceFeedback/FF_Generic_Reload' (0x8509b571f80) Class 'ForceFeedbackEffect'>"]
         [4989, "<Object '/GrillMoon/ForcedFeedback/GrillMoon_FF_Low_Rumble' (0x8509b572300) Class 'ForceFeedbackEffect'>"]
         [5312, "<Object '&GrillMoon_FF_Low_Rumble' (0x8509b572300) Class 'ForceFeedbackEffect'>"]
         [5575, "<Object '&GrillMoon_FF_Low_Rumble' (0x8509b572300) Class 'ForceFeedbackEffect'>"]
         [5789, "<Object '&GrillMoon_FF_Low_Rumble' (0x8509b572300) Class 'ForceFeedbackEffect'>"]
         [5953, "<Object '/FNGameplayCues/Weapons/ForceFeedback/FF_Generic_Whoosh' (0x8509b5751d0) Class 'ForceFeedbackEffect'>"]
         [6031, "<Object '/Game/Weapons/ForceFeedback/FF_Generic_TakeDamage' (0x84c3c0f3db0) Class 'ForceFeedbackEffect'>"]
         [6069, "<Object '/Game/Weapons/ForceFeedback/FF_Ostrich_Shield_Deactivate' (0x8509b575160) Class 'ForceFeedbackEffect'>"]
         [6102, "<Object '/Game/Weapons/ForceFeedback/FF_Ostrich_LoadMissile' (0x8509b5750f0) Class 'ForceFeedbackEffect'>"]
         [6202, "<Object '/Game/Weapons/ForceFeedback/FF_MissleLaunch' (0x85098d48ce0) Class 'ForceFeedbackEffect'>"]
         [6338, "<Object '/Game/Weapons/ForceFeedback/FF_Ostrich_FireMissile' (0x8509b572290) Class 'ForceFeedbackEffect'>"]
         [6436, "<Object '/FNGameplayCues/Weapons/ForceFeedback/FF_Generic_Quick_Jolt_Near' (0x8509b575240) Class 'ForceFeedbackEffect'>"]
         [6518, "<Object '&GrillMoon_FF_pjw1000_Demon_Charge_BuildUp' (0x8509b575080) Class 'ForceFeedbackEffect'>"]
        ]
       }]
     }]
   }
   {
    type=MovieSceneObjectPropertyTrack
    display_name=Cam Shake Base
    sections [{
      type=MovieSceneObjectPropertySection
      channels [None {
        type=MovieSceneScriptingObjectPathChannel
        default="<Object '&Events_Shared_00small_csb_C' (0x84c3c6a8a00) Class 'BlueprintGeneratedClass'>"
        keys [
         [
          0
          "<Object '&Events_Shared_00small_csb_C' (0x84c3c6a8a00) Class 'BlueprintGeneratedClass'>"
         ]
         [
          6031
          "<Object '/GrillMoon/CameraShake/GrillMoon_pjw1000_demonwalk_csb.GrillMoon_pjw1000_demonwalk_csb_C' (0x850a101c600) Class 'BlueprintGeneratedClass'>"
         ]
         [
          6518
          "<Object '/GrillMoon/CameraShake/GrillMoon_pjw1000_end_csb.GrillMoon_pjw1000_end_csb_C' (0x85051a35400) Class 'BlueprintGeneratedClass'>"
         ]
        ]
       }]
     }]
   }
   {
    type=MovieSceneEventTrack
    display_name=Cam Shake Radius Event Trigger
    sections [{
      type=MovieSceneEventTriggerSection
      channels [None {type=MovieSceneScriptingEventChannel, const ["<Struct 'MovieSceneEvent' (0x8502435a1e0) {}>", 1]}]
     }]
   }
   {type=MovieSceneSpawnTrack, sections [{type=MovieSceneSpawnSection, channels [None {type=MovieSceneScriptingBoolChannel, default=true}]}]}
   {
    type=MovieSceneEventTrack
    display_name=Force Feedback Radius Event Trigger
    sections [{
      type=MovieSceneEventTriggerSection
      channels [None {type=MovieSceneScriptingEventChannel, const ["<Struct 'MovieSceneEvent' (0x850243596a0) {}>", 1]}]
     }]
   }
   {
    type=MovieScene3DTransformTrack
    sections [{
      type=MovieScene3DTransformSection
      channels [
       Location.X {type=MovieSceneScriptingDoubleChannel, default=-57292.7491}
       Location.Y {type=MovieSceneScriptingDoubleChannel, default=41260.4837}
       Location.Z {type=MovieSceneScriptingDoubleChannel, default=10012.8182}
       Rotation.X {type=MovieSceneScriptingDoubleChannel, default=0}
       Rotation.Y {type=MovieSceneScriptingDoubleChannel, default=0}
       Rotation.Z {type=MovieSceneScriptingDoubleChannel, default=0}
       Scale.X {type=MovieSceneScriptingDoubleChannel, default=1}
       Scale.Y {type=MovieSceneScriptingDoubleChannel, default=1}
       Scale.Z {type=MovieSceneScriptingDoubleChannel, default=1}
      ]
     }]
   }
   {
    type=MovieSceneIntegerTrack
    sections [{type=MovieSceneIntegerSection, channels [None {type=MovieSceneScriptingIntegerChannel, default=2, const [0,1]}]}]
   }
   {
    type=MovieSceneEventTrack
    display_name=Cam Shake Lobby Event Trigger
    sections [{
      type=MovieSceneEventTriggerSection
      channels [None {
        type=MovieSceneScriptingEventChannel
        keys [
         [0, "<Struct 'MovieSceneEvent' (0x85024351cc0) {}>"]
         [6031, "<Struct 'MovieSceneEvent' (0x85024351ae0) {}>"]
         [6205, "<Struct 'MovieSceneEvent' (0x85024351900) {}>"]
         [6341, "<Struct 'MovieSceneEvent' (0x850243514a0) {}>"]
         [6518, "<Struct 'MovieSceneEvent' (0x850243512c0) {}>"]
        ]
       }]
     }]
   }
   {
    type=MovieSceneEventTrack
    display_name=Force Feedback Lobby Event Trigger
    sections [{
      type=MovieSceneEventTriggerSection
      channels [None {
        type=MovieSceneScriptingEventChannel
        keys [
         [0, "<Struct 'MovieSceneEvent' (0x85024350d20) {}>"]
         [4965, "<Struct 'MovieSceneEvent' (0x85024351220) {}>"]
         [4989, "<Struct 'MovieSceneEvent' (0x85024351360) {}>"]
         [5312, "<Struct 'MovieSceneEvent' (0x85024351a40) {}>"]
         [5575, "<Struct 'MovieSceneEvent' (0x85024351b80) {}>"]
         [5789, "<Struct 'MovieSceneEvent' (0x850243521c0) {}>"]
         [5953, "<Struct 'MovieSceneEvent' (0x85024352300) {}>"]
         [6031, "<Struct 'MovieSceneEvent' (0x85024352760) {}>"]
         [6069, "<Struct 'MovieSceneEvent' (0x850243528a0) {}>"]
         [6102, "<Struct 'MovieSceneEvent' (0x85024352d00) {}>"]
         [6202, "<Struct 'MovieSceneEvent' (0x85024352e40) {}>"]
         [6338, "<Struct 'MovieSceneEvent' (0x850243532a0) {}>"]
         [6436, "<Struct 'MovieSceneEvent' (0x850243533e0) {}>"]
         [6518, "<Struct 'MovieSceneEvent' (0x8502435d660) {}>"]
        ]
       }]
     }]
   }
  ]
 }]
display_rate {denominator=1, numerator=60}
tick_resolution {denominator=1, numerator=24000}
tracks []
