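# GrillMoon pjw1000 gameplay Level Sequence that places a timed knockback volume at a fixed point in the world.
# A spawn track keeps BP_GrillMoon_Knockback present for one window of the sequence.
# Constant tracks on its child KnockbackCapsule size it as a large, tall capsule and set its collision to query-only.
# A root event track fires two triggers within that window, with their bound functions not shown.
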LevelSequence /GrillMoon/Sequence/pjw/pjw1000/GrillMoon_pjw1000_gam
bindings [
 KnockbackCapsule {
  id=#1
  parent_id=#2
  possessed_class=CapsuleComponent
  tracks [
   {
    type=MovieSceneFloatTrack
    display_name=Capsule Half Height
    sections [{type=MovieSceneFloatSection, channels [None {type=MovieSceneScriptingFloatChannel, default=4531.0303, const [4531.0303,1]}]}]
   }
   {
    type=MovieSceneFloatTrack
    display_name=Capsule Radius
    sections [{type=MovieSceneFloatSection, channels [None {type=MovieSceneScriptingFloatChannel, default=800, const [800,1]}]}]
   }
   {
    type=MovieSceneByteTrack
    display_name="Collision Enabled (Body Instance)"
    sections [{type=MovieSceneByteSection, channels [None {type=MovieSceneScriptingByteChannel, default=1, const [1,1]}]}]
   }
  ]
 }
 BP_GrillMoon_Knockback {
  id=#2
  possessed_class=BP_GrillMoon_Knockback_C
  tracks [
   {
    type=MovieSceneSpawnTrack
    sections [{
      type=MovieSceneSpawnSection
      channels [None {type=MovieSceneScriptingBoolChannel, default=false, keys [[3478,false],[4134,true],[7348,false]]}]
     }]
   }
   {
    type=MovieScene3DTransformTrack
    sections [{
      type=MovieScene3DTransformSection
      channels [
       Location.X {type=MovieSceneScriptingDoubleChannel, default=-57590}
       Location.Y {type=MovieSceneScriptingDoubleChannel, default=42203.5834}
       Location.Z {type=MovieSceneScriptingDoubleChannel, default=9936}
       Rotation.X {type=MovieSceneScriptingDoubleChannel, default=0}
       Rotation.Y {type=MovieSceneScriptingDoubleChannel, default=0}
       Rotation.Z {type=MovieSceneScriptingDoubleChannel, default=0}
       Scale.X {type=MovieSceneScriptingDoubleChannel, default=1, const [1,2]}
       Scale.Y {type=MovieSceneScriptingDoubleChannel, default=1, const [1,2]}
       Scale.Z {type=MovieSceneScriptingDoubleChannel, default=1, const [1,2]}
      ]
     }]
   }
  ]
 }
]
display_rate {denominator=1, numerator=60}
tick_resolution {denominator=1, numerator=24000}
tracks [{
  type=MovieSceneEventTrack
  sections [{
    type=MovieSceneEventTriggerSection
    channels [None {
      type=MovieSceneScriptingEventChannel
      keys [[5348, "<Struct 'MovieSceneEvent' (0x8502435fe60) {}>"], [6023, "<Struct 'MovieSceneEvent' (0x8502435db60) {}>"]]
     }]
   }]
 }]
