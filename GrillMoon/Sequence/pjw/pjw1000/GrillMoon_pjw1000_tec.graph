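# Level Sequence for GrillMoon shot pjw1000 that toggles the visibility of the head mesh on a ClayTile Dragon Gate actor.
# It possesses the BSG_ClayTile_DragonGate_B_GrillMoon actor and binds its SM_Head static mesh component as a child.
# A single visibility track on that component keys the head visible, then hidden later in the sequence.
# The actor binding itself has no tracks, and the sequence has no root tracks.
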LevelSequence /GrillMoon/Sequence/pjw/pjw1000/GrillMoon_pjw1000_tec
bindings [
 SM_Head {
  id=#1
  parent_id=#2
  possessed_class=StaticMeshComponent
  tracks [{
    type=MovieSceneVisibilityTrack
    sections [{type=MovieSceneVisibilitySection, channels [None {type=MovieSceneScriptingBoolChannel, default=true, keys [[3503,true],[5328,false]]}]}]
   }]
 }
 BSG_ClayTile_DragonGate_B_GrillMoon {id=#2, possessed_class=BSG_ClayTile_DragonGate_GrillMoon_C, tracks []}
]
display_rate {denominator=1, numerator=60}
tick_resolution {denominator=1, numerator=24000}
tracks []
