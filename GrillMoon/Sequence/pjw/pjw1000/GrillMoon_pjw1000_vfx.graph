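# VFX level sequence for GrillMoon cinematic pjw1000, apparently a demon being pulled out of a dragon gate portal.
# Timed spawn and Niagara system tracks drive the pull, two demon silhouette effects and the portal exit burst, with all but the exit effect attached.
# The pull effect's rope whip, waveform distance and fade-out are animated through Niagara parameter tracks.
# The placed DragonGate Blueprint has component visibility tracks, hiding its head mesh partway through.
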LevelSequence /GrillMoon/Sequence/pjw/pjw1000/GrillMoon_pjw1000_vfx
bindings [
 "NS_GrillMoon_Demon_Silhouette (2)" {
  id=#1
  possessed_class=NiagaraActor
  tracks [
   {
    type=MovieSceneSpawnTrack
    sections [{type=MovieSceneSpawnSection, channels [None {type=MovieSceneScriptingBoolChannel, default=false, keys [[5606,false],[5625,true]]}]}]
   }
   {
    type=MovieScene3DTransformTrack
    sections [{
      type=MovieScene3DTransformSection
      channels [
       Location.X {type=MovieSceneScriptingDoubleChannel, default=60.2994, const [60.2994,1]}
       Location.Y {type=MovieSceneScriptingDoubleChannel, default=58.4807, const [58.4807,1]}
       Location.Z {type=MovieSceneScriptingDoubleChannel, default=0, const [0,1]}
       Rotation.X {type=MovieSceneScriptingDoubleChannel, default=0, const [0,1]}
       Rotation.Y {type=MovieSceneScriptingDoubleChannel, default=0, const [0,1]}
       Rotation.Z {type=MovieSceneScriptingDoubleChannel, default=0, const [0,1]}
       Scale.X {type=MovieSceneScriptingDoubleChannel, default=1, const [1,1]}
       Scale.Y {type=MovieSceneScriptingDoubleChannel, default=1, const [1,1]}
       Scale.Z {type=MovieSceneScriptingDoubleChannel, default=1, const [1,1]}
      ]
     }]
   }
   {type=MovieScene3DAttachTrack, sections [{type=MovieScene3DAttachSection, end_frame=12085, start_frame=3388}]}
  ]
 }
 VFX_Gate {
  id=#2
  parent_id=#3
  possessed_class=StaticMeshComponent
  tracks [{
    type=MovieSceneVisibilityTrack
    sections [{type=MovieSceneVisibilitySection, channels [None {type=MovieSceneScriptingBoolChannel, default=true}]}]
   }]
 }
 NS_GrillMoon_DemonExitsPortal {
  id=#4
  possessed_class=NiagaraActor
  tracks [
   {
    type=MovieSceneSpawnTrack
    sections [{type=MovieSceneSpawnSection, channels [None {type=MovieSceneScriptingBoolChannel, default=false, keys [[5928,false],[6019,true]]}]}]
   }
   {
    type=MovieScene3DTransformTrack
    sections [{
      type=MovieScene3DTransformSection
      channels [
       Location.X {type=MovieSceneScriptingDoubleChannel, default=-57594.2109}
       Location.Y {type=MovieSceneScriptingDoubleChannel, default=42656.1797}
       Location.Z {type=MovieSceneScriptingDoubleChannel, default=9943.0403}
       Rotation.X {type=MovieSceneScriptingDoubleChannel, default=0}
       Rotation.Y {type=MovieSceneScriptingDoubleChannel, default=0}
       Rotation.Z {type=MovieSceneScriptingDoubleChannel, default=0}
       Scale.X {type=MovieSceneScriptingDoubleChannel, default=1}
       Scale.Y {type=MovieSceneScriptingDoubleChannel, default=1}
       Scale.Z {type=MovieSceneScriptingDoubleChannel, default=1}
      ]
     }]
   }
  ]
 }
 BP_ClayTile_DragonGate_B2 {
  id=#3
  possessed_class=BSG_ClayTile_DragonGate_B_C
  tracks [{
    type=MovieScene3DTransformTrack
    sections [{
      type=MovieScene3DTransformSection
      channels [
       Location.X {type=MovieSceneScriptingDoubleChannel, default=-57600}
       Location.Y {type=MovieSceneScriptingDoubleChannel, default=42904}
       Location.Z {type=MovieSceneScriptingDoubleChannel, default=9988}
       Rotation.X {type=MovieSceneScriptingDoubleChannel, default=0}
       Rotation.Y {type=MovieSceneScriptingDoubleChannel, default=0}
       Rotation.Z {type=MovieSceneScriptingDoubleChannel, default=180}
       Scale.X {type=MovieSceneScriptingDoubleChannel, default=2}
       Scale.Y {type=MovieSceneScriptingDoubleChannel, default=2}
       Scale.Z {type=MovieSceneScriptingDoubleChannel, default=2}
      ]
     }]
   }]
 }
 NiagaraComponent0 {
  id=#5
  parent_id=#6
  possessed_class=NiagaraComponent
  tracks [
   {type=MovieSceneNiagaraSystemTrack, sections [{type=MovieSceneNiagaraSystemSpawnSection, end_frame=6023, start_frame=5310}]}
   {
    type=MovieSceneNiagaraFloatParameterTrack
    display_name=WaveformDistance
    sections [{
      type=MovieSceneFloatSection
      channels [None {
        type=MovieSceneScriptingFloatChannel
        default=100
        keys [[4871,0],[5180,50],[5233,20],[5310,0],[5353,100],[5474,0],[5491,50],[5515,100],[5625,50],[5675,0],[5859,100],[5916,0]]
       }]
     }]
   }
   {
    type=MovieSceneNiagaraBoolParameterTrack
    display_name=WhipRope
    sections [{type=MovieSceneBoolSection, channels [None {type=MovieSceneScriptingBoolChannel, default=true, keys [[4871,false],[5859,true]]}]}]
   }
   {
    type=MovieSceneNiagaraBoolParameterTrack
    display_name=FadeOutFX
    sections [{type=MovieSceneBoolSection, channels [None {type=MovieSceneScriptingBoolChannel, default=false, keys [[5886,false],[5927,true]]}]}]
   }
  ]
 }
 NS_GrillMoon_Demon_Silhouette {
  id=#7
  possessed_class=NiagaraActor
  tracks [
   {
    type=MovieSceneSpawnTrack
    sections [{type=MovieSceneSpawnSection, channels [None {type=MovieSceneScriptingBoolChannel, default=true, keys [[5460,false],[5933,true]]}]}]
   }
   {
    type=MovieScene3DTransformTrack
    sections [{
      type=MovieScene3DTransformSection
      channels [
       Location.X {type=MovieSceneScriptingDoubleChannel, default=0}
       Location.Y {type=MovieSceneScriptingDoubleChannel, default=0}
       Location.Z {type=MovieSceneScriptingDoubleChannel, default=0}
       Rotation.X {type=MovieSceneScriptingDoubleChannel, default=0}
       Rotation.Y {type=MovieSceneScriptingDoubleChannel, default=0}
       Rotation.Z {type=MovieSceneScriptingDoubleChannel, default=0}
       Scale.X {type=MovieSceneScriptingDoubleChannel, default=1}
       Scale.Y {type=MovieSceneScriptingDoubleChannel, default=1}
       Scale.Z {type=MovieSceneScriptingDoubleChannel, default=1}
      ]
     }]
   }
   {type=MovieScene3DAttachTrack, sections [{type=MovieScene3DAttachSection, end_frame=12125, start_frame=3428}]}
  ]
 }
 NE_SystemShare {
  id=#8
  parent_id=#3
  possessed_class=NiagaraComponent
  tracks [{
    type=MovieSceneVisibilityTrack
    sections [{type=MovieSceneVisibilitySection, channels [None {type=MovieSceneScriptingBoolChannel, default=true}]}]
   }]
 }
 NiagaraComponent0 {
  id=#9
  parent_id=#7
  possessed_class=NiagaraComponent
  tracks [{type=MovieSceneNiagaraSystemTrack, sections [{type=MovieSceneNiagaraSystemSpawnSection, end_frame=6057, start_frame=5933}]}]
 }
 SM_Head {
  id=#10
  parent_id=#3
  possessed_class=StaticMeshComponent
  tracks [{
    type=MovieSceneVisibilityTrack
    sections [{type=MovieSceneVisibilitySection, channels [None {type=MovieSceneScriptingBoolChannel, default=false, keys [[3503,true],[5025,false]]}]}]
   }]
 }
 NiagaraComponent0 {
  id=#11
  parent_id=#1
  possessed_class=NiagaraComponent
  tracks [{type=MovieSceneNiagaraSystemTrack, sections [{type=MovieSceneNiagaraSystemSpawnSection, end_frame=5933, start_frame=5721}]}]
 }
 NS_GrillMoon_PulledfromPortal {
  id=#6
  possessed_class=NiagaraActor
  tracks [
   {type=MovieSceneSpawnTrack, sections [{type=MovieSceneSpawnSection, channels [None {type=MovieSceneScriptingBoolChannel, default=true}]}]}
   {
    type=MovieScene3DTransformTrack
    sections [{
      type=MovieScene3DTransformSection
      channels [
       Location.X {type=MovieSceneScriptingDoubleChannel, default=0, const [0,1]}
       Location.Y {type=MovieSceneScriptingDoubleChannel, default=224, const [224,1]}
       Location.Z {type=MovieSceneScriptingDoubleChannel, default=0, const [0,1]}
       Rotation.X {type=MovieSceneScriptingDoubleChannel, default=0, const [0,1]}
       Rotation.Y {type=MovieSceneScriptingDoubleChannel, default=0, const [0,1]}
       Rotation.Z {type=MovieSceneScriptingDoubleChannel, default=0, const [0,1]}
       Scale.X {type=MovieSceneScriptingDoubleChannel, default=1, const [1,1]}
       Scale.Y {type=MovieSceneScriptingDoubleChannel, default=1, const [1,1]}
       Scale.Z {type=MovieSceneScriptingDoubleChannel, default=1, const [1,1]}
      ]
     }]
   }
   {type=MovieScene3DAttachTrack, sections [{type=MovieScene3DAttachSection, end_frame=11781, start_frame=3081}]}
  ]
 }
]
display_rate {denominator=1, numerator=60}
tick_resolution {denominator=1, numerator=24000}
tracks []
